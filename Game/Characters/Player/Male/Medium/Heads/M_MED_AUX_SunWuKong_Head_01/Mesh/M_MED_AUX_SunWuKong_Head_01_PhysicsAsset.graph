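# Physics asset for the SunWuKong head mesh of a medium-build male player character.
# It has four capsule bodies on the upper spine, neck, face attach and jaw bones.
# Three constraints link them in a chain from spine_05 through neck_02 and faceAttach to C_jaw, each with matching swing and twist limit angles.
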PhysicsAsset /Game/Characters/Player/Male/Medium/Heads/M_MED_AUX_SunWuKong_Head_01/Mesh/M_MED_AUX_SunWuKong_Head_01_PhysicsAsset
bodies {
 C_jaw [{shapeType=Capsule, length=8.9362, radius=7.4119, center=<-0.8829 3.1516 -3.354>, rotation=<rot -33.169 -90 -57.7988>}]
 faceAttach [{shapeType=Capsule, length=6.0453, radius=5.2541, center=<1.7695 6.6908 1.8474>, rotation=<rot -3.0279 -90 -41.9996>}]
 neck_02 [{shapeType=Capsule, length=16.104, radius=10.4137, center=<13.3743 -0.0444 -0.031>, rotation=<rot 0 -82.5754 -89.7517>}]
 spine_05 [{shapeType=Capsule, length=10.1649, radius=9.2145, center=<9.687 1.3426 -0.5514>, rotation=<rot 0 20.3122 -169.1718>}]
}
constraints [
 {bone1Name=neck_02, bone2Name=spine_05, swing1LimitDegrees=45, swing2LimitDegrees=45, twistLimitDegrees=45}
 {bone1Name=faceAttach, bone2Name=neck_02, swing1LimitDegrees=45, swing2LimitDegrees=45, twistLimitDegrees=45}
 {bone1Name=C_jaw, bone2Name=faceAttach, swing1LimitDegrees=45, swing2LimitDegrees=45, twistLimitDegrees=45}
]
